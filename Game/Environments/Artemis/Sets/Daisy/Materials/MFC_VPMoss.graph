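# Material function that layers vertex-painted moss over an input material, toggled by the EnableMossPainting static switch.
# World-aligned moss textures drive curve-atlas color, specular, roughness and normal, with a Fresnel term shading darker cores and brighter fuzzy edges.
# Red vertex color height-blends the moss against the inverted TrunkHeight input, optionally including the moss heightmap.
# The mobile path overrides only base color, using flat darkening and a parameter-driven curve lookup.
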MaterialFunction /Game/Environments/Artemis/Sets/Daisy/Materials/MFC_VPMoss
graph {nodes [
  n1 = GetMaterialAttributes(None=FunctionInput(Preview=MakeMaterialAttributes!(
     BaseColor
     Metallic
     Specular
     Roughness
     Anisotropy
     EmissiveColor
     Opacity
     OpacityMask
     Normal
     Tangent
     WorldPositionOffset
     SubsurfaceColor
     ClearCoat
     ClearCoatRoughness
     AmbientOcclusion
     Refraction
     CustomizedUVs_0
     CustomizedUVs_1
     CustomizedUVs_2
     CustomizedUVs_3
     CustomizedUVs_4
     CustomizedUVs_5
     CustomizedUVs_6
     CustomizedUVs_7
     PixelDepthOffset
     ShadingModel
     Displacement
    )) {bUsePreviewValueAsDefault=true, inputName=Material, inputType=FunctionInput_MaterialAttributes}) {attributeGetTypes [#1]}
  n2 = Constant() {r=0.01}
  n3 = NamedRerouteUsage()
  n4 = VectorParameter() {group=Moss, parameterName=MossEdgeColor, defaultValue=<rgba 0.6215 0.3421 0.6771 1>}
  n5 = Saturate(None=Add(A=Multiply(A=@n2, B=@n3), B=Multiply(A=@n3, B=@n4.RGB)))
  n6 = ScalarParameter() {defaultValue=0.25, group=Moss, parameterName=Moss_CoreDarkening}
  n7 = ScalarParameter() {defaultValue=1, group=Moss, parameterName=Moss_FuzzyEdgeBrightness}
  n8 = TextureObjectParameter() {
   group=Moss
   parameterName=Normals Top
   samplerType=SAMPLERTYPE_Normal
   sortPriority=6
   texture=/Game/Environments/Artemis/Sets/Terrain/Textures/T_Artemis_Frontier_Moss_01_N
  }
  n9 = ScalarParameter() {defaultValue=500, group=Moss, parameterName=Scale Top, sortPriority=3}
  n10 = MaterialFunctionCall(TextureObject=@n8, TextureSize=@n9, WorldSpace=StaticBool(), Normal=VertexNormalWS()) {materialFunction=&WorldAlignedNormal}
  n11 = ComponentMask(None=VertexNormalWS()) {b=true}
  n12 = MaterialFunctionCall(
   Normal=LinearInterpolate(A=@n10.XYZFlatTop, B=@n10.XYZFlatTop, Alpha=Clamp(None=@n11))
   Flatness=ScalarParameter() {group=Moss, parameterName=MossNormalFlatten}
  ) {materialFunction=&FlattenNormal}
  n13 = Power(
   Base=Fresnel(ExponentIn=ScalarParameter() {defaultValue=2, group=Moss, parameterName=FuzzExponent}, Normal=Transform(None=@n12.Result))
   Exp=ScalarParameter() {defaultValue=2, group=Moss, parameterName=FuzzEdgeClamp}
  )
  n14 = ShadingPathSwitch(
   Default=LinearInterpolate(A=Multiply(A=@n5, B=@n6), B=Multiply(A=@n5, B=@n7), Alpha=Clamp(None=@n13))
   Mobile=Multiply(A=@n5, B=ScalarParameter() {defaultValue=0.6, group=Moss, parameterName=Mobile_Moss_Darkening})
  )
  n15 = StaticBool() {value=true}
  n16 = TextureObjectParameter() {
   group=Top
   parameterName=Spec Top
   samplerType=SAMPLERTYPE_Masks
   sortPriority=4
   texture=/Game/Environments/Artemis/Sets/Terrain/Textures/T_Artemis_Frontier_Moss_01_S
  }
  n17 = WorldPosition()
  n18 = MaterialFunctionCall(Export Float 4=@n15, TextureObject=@n16, TextureSize=@n9, WorldPosition=@n17.XYZ) {materialFunction=/Game/Weapons/WeaponSkins/Wraps/MaterialFunctions/WorldAlignedTexture}
  n19 = SetMaterialAttributes!(
   MaterialAttributes=@n1.MaterialAttributes
   BaseColor=@n14
   Specular=LinearInterpolate(
    A=ScalarParameter() {defaultValue=0.1, group=Moss, parameterName=Moss_Min_Specular}
    B=ScalarParameter() {defaultValue=0.5, group=Moss, parameterName=Moss_Max_Specular}
    Alpha=ComponentMask(None=@n18.XYZ Texture) {r=true}
   )
   Roughness=LinearInterpolate(
    A=ScalarParameter() {defaultValue=0.75, group=Moss, parameterName=Moss_Min_Roughness}
    B=ScalarParameter() {defaultValue=1, group=Moss, parameterName=Moss_Max_Roughness}
    Alpha=ComponentMask(None=@n18.XYZ Texture) {b=true}
   )
   Normal=@n12.Result
  ) {attributeSetTypes [
    #2
    #3
    #4
    #1
   ]}
  n20 = ShadingPathSwitch(
   Default=@n19
   Mobile=SetMaterialAttributes!(MaterialAttributes=@n1.MaterialAttributes, BaseColor=@n14) {attributeSetTypes [#2]}
  )
  n21 = VertexColor()
  n22 = NamedRerouteUsage()
  n23 = OneMinus(None=FunctionInput(Preview=Constant()) {bUsePreviewValueAsDefault=true, inputName=TrunkHeight, inputType=FunctionInput_Scalar})
  n24 = ScalarParameter() {group=Moss, parameterName=VertexPaintContrast}
  n25 = MaterialFunctionCall!(
   Diffuse 1
   Diffuse 2
   Transistion Phase=@n21.R
   Height Texture 1=Multiply(A=@n22, B=ScalarParameter() {defaultValue=2, group=Moss, parameterName=MossHeightInfluence}) {constB=0.5}
   Height Texture 2=@n23
   Contrast=@n24
  ) {materialFunction=/Engine/Functions/Engine_MaterialFunctions02/Texturing/HeightLerpWithTwoHeightMaps}
  n26 = MaterialFunctionCall(Transition Phase=@n21.R, Height Texture=@n23, Contrast=@n24) {materialFunction=&HeightLerp}
  n27 = StaticSwitchParameter(True=@n25.Alpha, False=@n26.Alpha) {group=Moss, parameterName=BlendMossheightmap}
  FunctionOutput(None=StaticSwitchParameter(True=BlendMaterialAttributes(A=@n1.MaterialAttributes, B=@n20, Alpha=@n27), False=@n1.MaterialAttributes) {parameterName=EnableMossPainting})
  n28 = ScalarParameter() {group=Moss, parameterName=MossColorEdge_1}
  n29 = ScalarParameter() {defaultValue=2, group=Moss, parameterName=MossColorEdge_2}
  n30 = StaticSwitchParameter(True=@n25.Lerp Alpha No Contrast, False=@n26.Lerp Alpha No Contrast) {group=Moss, parameterName=BlendMossheightmap}
  LinearInterpolate(
   A=VectorParameter() {parameterName=MossEdgeColor, defaultValue=<rgba 0.6215 0.3421 0.6771 1>}.RGB
   B=@n3
   Alpha=Saturate(None=LinearInterpolate(A=@n28, B=@n29, Alpha=@n30))
  )
  n31 = TextureObjectParameter() {
   group=Moss
   parameterName=Diffuse Top
   samplerType=SAMPLERTYPE_Masks
   sortPriority=4
   texture=/Game/Environments/Artemis/Sets/Terrain/Textures/T_Artemis_Frontier_Moss_01_D
  }
  n32 = MaterialFunctionCall(Export Float 4=@n15, TextureObject=@n31, TextureSize=@n9, WorldPosition=@n17.XYZ) {materialFunction=&WorldAlignedTexture}
  NamedRerouteDeclaration(None=ComponentMask(None=@n32.XYZ Texture) {a=true}) {name=MossHeight, nodeColor=<rgba 0 0.8594 1 1>}
  n33 = ScalarParameter() {defaultValue=0.15, group=Moss, parameterName=MossMinColVal}
  n34 = ComponentMask(None=@n32.XYZ Texture) {r=true}
  n35 = CurveAtlasRowParameter(CurveTime=ShadingPathSwitch(Default=LinearInterpolate(A=@n33, Alpha=@n34) {constA=0.25}, Mobile=ScalarParameter() {parameterName=MobileCurveTime})) {
   atlas=/Game/Athena/Apollo/Environments/Landscape/Materials/Frontier/FrontierColorAtlas
   curve=/Game/Athena/Apollo/Environments/Landscape/Materials/Frontier/FrontierMossDetailColor
   group=Moss
   parameterName=MossColorCurve
  }
  NamedRerouteDeclaration(None=@n35) {name=MossBaseColorDetail, nodeColor=<rgba 1 0 0.6562 1>}
  OneMinus(None=@n22)
  Reroute()
  SetMaterialAttributes!(MaterialAttributes=@n1.MaterialAttributes)
  Transform(None=Constant3Vector() {constant=<rgba 0 0 1 0>}) {transformSourceType=TRANSFORMSOURCE_World, transformType=TRANSFORM_Tangent}
 ]}
